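# Niagara system for the CirrusVine glider's green thruster effects and wing trails, attached to the glider's skeletal mesh.
# Three mesh emitters form the thruster plumes, partitioned sprites spread along two mirrored lines, and a looping mesh fakes lighting.
# Four ribbon anim trails between bone or socket pairs trace the side wings.
# A system Timeline, sometimes added to a size-scale value, ramps brightness, opacity and plume length, with front-end overrides on some values.
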NiagaraSystem /BRCosmetics/Gadgets/Assets/VinderTech_GliderChute/Glider_CirrusVine/FX/NS_Glider_CirrusVine_Thrusters
dependencies {
 usedDataInterfaces [
  &NiagaraDataInterfaceCurve
  &NiagaraDataInterfaceColorCurve
  &NiagaraDataInterfaceArrayFloat
  &NiagaraDataInterfaceArrayFloat3
 ]
 usedDynamicInputs [
  &DistanceBetweenPositions
  &Boolean_AndOperation
  &SetBoolByFloatComparison
 ]
 usedModules [
  &SystemState
  &Timeline
  &self:SystemSpawnScript.SystemScriptSource.SystemScriptGraph.NiagaraNodeAssignment_0.SetVariables_#1
 ]
}
emitters {
 AirThrusters {
  emitterUpdateScript [
   &EmitterState()
   &SpawnBurst_Instantaneous(Spawn Count=2, Use Spawn Probability=false, UseLoopCountLimit=false) {enabled=false}
   &SpawnRate(SpawnRate=10, Use Spawn Probability=false)
  ]
  particleSpawnScript [
   &PartitionParticles() {enabled=false}
   &InitializeParticle(Color=<rgba 0.3294 1 0.6039 0.1>, Mesh Scale Mode=::NewEnumerator4, UsePositionOffset=false, Write Material Random=true)
   &InitialMeshOrientation(Mesh Axis To Orient=<0 1 0>, Mesh Orientation Mode=::NewEnumerator2, Orientation Vector=<0 1 0>, Use Rotation Vector=true)
   &DynamicMaterialParameters(
    Index 0 Param 1=&RandomRangeFloat(Minimum=1, Maximum=1)
    Index 0 Param 2=16
    Module.Param0WriteEnabled=true
    Module.Param1WriteEnabled=true
    Module.Param2WriteEnabled=false
    Module.Param3WriteEnabled=false
   )
  ]
  particleUpdateScript [
   &ParticleState()
   &SkeletalMeshLocation(
    MaskSpawnGroup=false
    Mesh Sampling Type=::NewEnumerator5
    Sampled Position Offset=<0 6 0>
    Skeletal Mesh=$User.SkeletalMesh
    Socket Sampling Mode=::NewEnumerator5
   )
   &ScaleMeshSize(Scale Factor=&FN_FrontEndVector3Modifier(
     Gameplay=&MakeVector(
      X=1
      Y=&Lerp_Float(
       A=0
       B=1
       Alpha=&ClampFloat(Float=&Add_Float(A=$System.Timeline.NormalizedTime, B=$System.Size scale))
      )
      Z=1
     )
     Front End=&Multiply_VectorByFloat(Vector=<1 1 1>, Float=0.64)
    )) {module=ScaleMeshSize001}
   &ScaleColor(
    Scale Alpha=&Multiply_Float(
     A=&FN_FrontEndFloatModifier(Gameplay=&Lerp_Float(A=0, B=1, Alpha=$System.Timeline.NormalizedTime), Front End=1)
     B=&Lerp_Float(A=0, B=0.2, Alpha=$Particles.NormalizedAge)
    )
    Scale RGB=&VectorFromFloat(Value=&Lerp_Float(
      A=0
      B=1.2
      Alpha=&ClampFloat(Float=&Add_Float(A=$System.Size scale, B=$System.Timeline.NormalizedTime))
     ))
    ScaleA=true
    ScaleRGB=true
   )
   /SetVariables(Particles.ColorAlt:LinearColor=&MakeLinearColorFromVectorAndFloat(
     "Vector (RGB)"=&MakeVectorFromLinearColorRGB(Color=&ColorFromCurve(ColorCurve=DataInterface() {propertyValues {
         AlphaCurve {keys [<at 0 1 RCIM_Cubic>,<at 1 1 RCIM_Cubic>]}
         BlueCurve {keys [
           <at 0.083 1 RCIM_Cubic>
           {arriveTangent=-2.419, interpMode=RCIM_Cubic, time=0.2491, value=0.1}
           {interpMode=RCIM_Cubic, leaveTangent=1.4033, time=0.5429, value=0.0469}
          ]}
         GreenCurve {keys [
           <at 0.083 0 RCIM_Cubic>
           {arriveTangent=0.0614, interpMode=RCIM_Cubic, time=0.2491, value=0.0565}
           {interpMode=RCIM_Cubic, leaveTangent=0.7928, time=0.5429, value=0.0242}
          ]}
         RedCurve {keys [<at 0.083 0.5929 RCIM_Cubic>, {arriveTangent=-1.5047, interpMode=RCIM_Cubic, time=0.2491, value=0}, <at 0.5429 0 RCIM_Cubic>]}
        }}))
     "Float (Alpha)"=&Lerp_Float(A=0, B=0.05, Alpha=$Particles.NormalizedAge)
    )) {enabled=false, isSetParameters=true, module=SetVariables_#2}
  ]
  rendererClasses [&NiagaraMeshRendererProperties]
  renderers [
   {rendererClass=&NiagaraMeshRendererProperties, rendererIndex=0}
   {rendererClass=&NiagaraMeshRendererProperties, rendererIndex=1}
  ]
 }
 AirThrusters001 {
  emitterUpdateScript [
   &EmitterState()
   &SpawnBurst_Instantaneous(Spawn Count=2, Use Spawn Probability=false, UseLoopCountLimit=false) {enabled=false}
   &SpawnRate(SpawnRate=10, Use Spawn Probability=false)
  ]
  particleSpawnScript [
   &PartitionParticles() {enabled=false}
   &InitializeParticle(Color=<rgba 0.3294 1 0.6039 0.1>, Mesh Scale Mode=::NewEnumerator4, UsePositionOffset=false, Write Material Random=true)
   &InitialMeshOrientation(Mesh Axis To Orient=<0 1 0>, Mesh Orientation Mode=::NewEnumerator2, Orientation Vector=<0 1 0>, Use Rotation Vector=true)
   &DynamicMaterialParameters(
    Index 0 Param 1=&RandomRangeFloat(Minimum=0.3, Maximum=5)
    Index 0 Param 2=10
    Module.Param0WriteEnabled=true
    Module.Param1WriteEnabled=true
    Module.Param2WriteEnabled=false
    Module.Param3WriteEnabled=false
   )
  ]
  particleUpdateScript [
   &ParticleState()
   &SkeletalMeshLocation(
    MaskSpawnGroup=false
    Mesh Sampling Type=::NewEnumerator5
    Sampled Position Offset=<0 6 0>
    Skeletal Mesh=$User.SkeletalMesh
    Socket Sampling Mode=::NewEnumerator5
   )
   &ScaleMeshSize(Scale Factor=&FN_FrontEndVector3Modifier(
     Gameplay=&MakeVector(
      X=1
      Y=&Lerp_Float(
       A=0
       B=1
       Alpha=&ClampFloat(Float=&Add_Float(A=$System.Timeline.NormalizedTime, B=$System.Size scale))
      )
      Z=1
     )
     Front End=&Multiply_VectorByFloat(Vector=<1 1 1>, Float=0.64)
    )) {module=ScaleMeshSize001}
   &ScaleColor(
    Scale Alpha=&Multiply_Float(
     A=&FN_FrontEndFloatModifier(Gameplay=&Lerp_Float(A=0, B=1, Alpha=$System.Timeline.NormalizedTime), Front End=1)
     B=&Lerp_Float(A=0, B=0.2, Alpha=$Particles.NormalizedAge)
    )
    Scale RGB=&VectorFromFloat(Value=&Lerp_Float(
      A=0
      B=5
      Alpha=&ClampFloat(Float=&Add_Float(A=$System.Size scale, B=$System.Timeline.NormalizedTime))
     ))
    ScaleA=true
    ScaleRGB=true
   )
   /SetVariables(Particles.ColorAlt:LinearColor=&MakeLinearColorFromVectorAndFloat(
     "Vector (RGB)"=&MakeVectorFromLinearColorRGB(Color=&ColorFromCurve(ColorCurve=DataInterface() {propertyValues {
         AlphaCurve {keys [<at 0 1 RCIM_Cubic>,<at 1 1 RCIM_Cubic>]}
         BlueCurve {keys [
           <at 0.083 1 RCIM_Cubic>
           {arriveTangent=-2.419, interpMode=RCIM_Cubic, time=0.2491, value=0.1}
           {interpMode=RCIM_Cubic, leaveTangent=1.4033, time=0.5429, value=0.0469}
          ]}
         GreenCurve {keys [
           <at 0.083 0 RCIM_Cubic>
           {arriveTangent=0.0614, interpMode=RCIM_Cubic, time=0.2491, value=0.0565}
           {interpMode=RCIM_Cubic, leaveTangent=0.7928, time=0.5429, value=0.0242}
          ]}
         RedCurve {keys [<at 0.083 0.5929 RCIM_Cubic>, {arriveTangent=-1.5047, interpMode=RCIM_Cubic, time=0.2491, value=0}, <at 0.5429 0 RCIM_Cubic>]}
        }}))
     "Float (Alpha)"=&Lerp_Float(A=0, B=0.05, Alpha=$Particles.NormalizedAge)
    )) {enabled=false, isSetParameters=true, module=SetVariables_#2}
   &DynamicMaterialParameters(
    Index 0 Param 2=&FloatFromCurve(FloatCurve=DataInterface() {propertyValues {Curve {keys [
         {interpMode=RCIM_Cubic, tangentMode=RCTM_User, time=0, value=1}
         {arriveTangent=-3.3934, interpMode=RCIM_Cubic, leaveTangent=-3.3934, tangentMode=RCTM_User, time=1, value=0}
        ]}}})
    Module.Param0WriteEnabled=false
    Module.Param1WriteEnabled=true
    Module.Param2WriteEnabled=false
    Module.Param3WriteEnabled=false
   ) {enabled=false, module=DynamicMaterialParameters001}
  ]
  rendererClasses [&NiagaraMeshRendererProperties]
  renderers [{rendererClass=&NiagaraMeshRendererProperties, rendererIndex=0}]
 }
 AirThrusters002 {
  emitterUpdateScript [
   &EmitterState()
   &SpawnBurst_Instantaneous(Spawn Count=2, Use Spawn Probability=false, UseLoopCountLimit=false) {enabled=false}
   &SpawnRate(SpawnRate=4, Use Spawn Probability=false)
  ]
  particleSpawnScript [
   &PartitionParticles() {enabled=false}
   &InitializeParticle(Color=<rgba 0.3294 1 0.6039 1>, Mesh Scale Mode=::NewEnumerator4, UsePositionOffset=false, Write Material Random=true)
   &InitialMeshOrientation(Mesh Axis To Orient=<0 1 0>, Mesh Orientation Mode=::NewEnumerator2, Orientation Vector=<0 1 0>, Use Rotation Vector=true)
   &DynamicMaterialParameters(
    Index 0 Param 1=&RandomRangeFloat(Minimum=0.8, Maximum=1)
    Index 0 Param 2=&RandomRangeFloat(Minimum=5.5, Maximum=6)
    Module.Param0WriteEnabled=true
    Module.Param1WriteEnabled=true
    Module.Param2WriteEnabled=false
    Module.Param3WriteEnabled=false
   )
  ]
  particleUpdateScript [
   &ParticleState()
   &SkeletalMeshLocation(
    MaskSpawnGroup=false
    Mesh Sampling Type=::NewEnumerator5
    Sampled Position Offset=<0 6 0>
    Skeletal Mesh=$User.SkeletalMesh
    Socket Sampling Mode=::NewEnumerator5
   )
   &ScaleMeshSize(Scale Factor=&FN_FrontEndVector3Modifier(Gameplay=<1 1 1>, Front End=&Multiply_VectorByFloat(Vector=<1 1 1>, Float=0.64))) {module=ScaleMeshSize001}
   &ScaleColor(
    Scale Alpha=&Multiply_Float(A=1, B=&Lerp_Float(A=0, B=1, Alpha=$Particles.NormalizedAge))
    Scale RGB=&VectorFromFloat(Value=&FN_FrontEndFloatModifier(
      Gameplay=&Lerp_Float(
       A=0
       B=5
       Alpha=&ClampFloat(Float=&Add_Float(A=$System.Size scale, B=$System.Timeline.NormalizedTime))
      )
      Front End=2
     ))
    ScaleA=true
    ScaleRGB=true
   )
   /SetVariables(Particles.ColorAlt:LinearColor=&MakeLinearColorFromVectorAndFloat(
     "Vector (RGB)"=&MakeVectorFromLinearColorRGB(Color=&ColorFromCurve(ColorCurve=DataInterface() {propertyValues {
         AlphaCurve {keys [<at 0 1 RCIM_Cubic>,<at 1 1 RCIM_Cubic>]}
         BlueCurve {keys [
           <at 0.083 1 RCIM_Cubic>
           {arriveTangent=-2.419, interpMode=RCIM_Cubic, time=0.2491, value=0.1}
           {interpMode=RCIM_Cubic, leaveTangent=1.4033, time=0.5429, value=0.0469}
          ]}
         GreenCurve {keys [
           <at 0.083 0 RCIM_Cubic>
           {arriveTangent=0.0614, interpMode=RCIM_Cubic, time=0.2491, value=0.0565}
           {interpMode=RCIM_Cubic, leaveTangent=0.7928, time=0.5429, value=0.0242}
          ]}
         RedCurve {keys [<at 0.083 0.5929 RCIM_Cubic>, {arriveTangent=-1.5047, interpMode=RCIM_Cubic, time=0.2491, value=0}, <at 0.5429 0 RCIM_Cubic>]}
        }}))
     "Float (Alpha)"=&Lerp_Float(A=0, B=0.05, Alpha=$Particles.NormalizedAge)
    )) {enabled=false, isSetParameters=true, module=SetVariables_#2}
  ]
  rendererClasses [&NiagaraMeshRendererProperties]
  renderers [{rendererClass=&NiagaraMeshRendererProperties, rendererIndex=0}]
 }
 AirThrusters003 {
  emitterUpdateScript [
   &EmitterState()
   &SpawnBurst_Instantaneous(Spawn Count=22, Use Spawn Probability=false, UseLoopCountLimit=false)
   &SpawnRate(SpawnRate=14, Use Spawn Probability=false)
   &SpawnBurst_Instantaneous(Spawn Count=12, Spawn Time=3.1884, Use Spawn Probability=false, UseLoopCountLimit=false) {enabled=false, module=SpawnBurst_Instantaneous001}
   &SpawnBurst_Instantaneous(Spawn Count=12, Spawn Time=1.01, Use Spawn Probability=false, UseLoopCountLimit=false) {enabled=false, module=SpawnBurst_Instantaneous002}
  ]
  particleSpawnScript [
   &PartitionParticles(Partition Mode:ENiagara_PartitionMode=::NewEnumerator1)
   &InitializeParticle(
    Color=<rgba 0.9882 3 1.8118 0.05>
    Lifetime Max=2.5
    Lifetime Min=1.5
    Lifetime Mode=::NewEnumerator1
    Sprite Size Mode=::NewEnumerator3
    Uniform Sprite Size=8
    UsePositionOffset=false
    Write Material Random=true
   )
   /SetVariables(Particles.particleIndex:NiagaraFloat=&SelectFloatFromArray(
     Array Sampling Mode=::NewEnumerator0
     Direct Array Index=&ReturnParticleIDAttributes()
     Direct Array Mode=::NewEnumerator1
     Float Selection Array=DataInterface() {propertyValues {FloatData [0,0.1,0.2,0.3,0.4,0.5,0.6,0.7,0.8,0.9,1]}}
     Random Seed=0
    )) {isSetParameters=true}
   &SpriteFacingAndAlignment(
    Sprite Facing=&SelectVectorFromArray(
     Direct Array Index=$Particles.PartitionParticles.Partition
     Vector Selection Array=DataInterface() {propertyValues {FloatData [<0.5 1 0>, <-0.5 1 0>]}}
    )
    WriteAlignment=false
    WriteFacing=true
   )
  ]
  particleUpdateScript [
   &ParticleState()
   &SkeletalMeshLocation(MaskSpawnGroup=false, Mesh Sampling Type=::NewEnumerator5, Skeletal Mesh=$User.SkeletalMesh, Socket Sampling Mode=::NewEnumerator5)
   &ShapeLocation(
    Cylinder Distribution=::NewEnumerator2
    Cylinder Height=30
    Cylinder Radius=0
    MaskSpawnGroup=false
    Offset=&SelectVectorFromArray(
     Direct Array Index=$Particles.PartitionParticles.Partition
     Vector Selection Array=DataInterface() {propertyValues {FloatData [<23 -4.7 0>, <-23 -4.7 0>]}}
    )
    Offset Mode=::NewEnumerator0
    Rotation Axis=&SelectVectorFromArray(
     Direct Array Index=$Particles.PartitionParticles.Partition
     Vector Selection Array=DataInterface() {propertyValues {FloatData [<1 -0.45 0>, <1 0.45 0>]}}
    )
    Rotation Mode=::NewEnumerator1
    Shape Origin=$Output.SkeletalMeshLocation.SampledPositionWithoutOffset
    Shape Primitive=::NewEnumerator1
    UseOwnerScale=true
    V Position=&Lerp_Float(A=0, B=1, Alpha=$Particles.particleIndex)
   )
   &ScaleColor(
    Scale Alpha=&Multiply_Float(
     A=&FloatFromCurve(FloatCurve=DataInterface() {propertyValues {Curve {keys [
          {arriveTangent=3.2444, interpMode=RCIM_Cubic, leaveTangent=3.2444, tangentMode=RCTM_User, time=0, value=0}
          <at 0.2 1 RCIM_Cubic>
          {arriveTangent=-1.25, interpMode=RCIM_Cubic, leaveTangent=-1.25, time=0.75, value=0.8}
          {arriveTangent=-4.31, interpMode=RCIM_Cubic, leaveTangent=-4.31, tangentMode=RCTM_User, time=1, value=0}
         ]}}})
     B=&Lerp_Float(A=0, B=1, Alpha=$System.Timeline.NormalizedTime)
    )
    Scale RGB=&VectorFromFloat(Value=&Multiply_Float(
      A=&FloatFromCurve(
       FloatCurve=DataInterface() {propertyValues {Curve {keys [
           {arriveTangent=6.8336, interpMode=RCIM_Cubic, leaveTangent=6.8336, tangentMode=RCTM_User, time=0, value=0}
           <at 0.2 1 RCIM_Cubic>
           {arriveTangent=-1.25, interpMode=RCIM_Cubic, leaveTangent=-1.25, time=0.75, value=0.8}
           {arriveTangent=-4.31, interpMode=RCIM_Cubic, leaveTangent=-4.31, tangentMode=RCTM_User, time=1, value=0}
          ]}}}
       Scale Curve=5
      )
      B=&Lerp_Float(A=0, B=1, Alpha=$System.Timeline.NormalizedTime)
     ))
    ScaleA=true
    ScaleRGB=true
   )
   &DynamicMaterialParameters(Write Parameter Index 0=false) {enabled=false}
  ]
  rendererClasses [&NiagaraSpriteRendererProperties]
  renderers [{rendererClass=&NiagaraSpriteRendererProperties}]
 }
 FakeLighting {
  emitterUpdateScript [
   &EmitterState()
   &SpawnBurst_Instantaneous(Spawn Count=4, Use Spawn Probability=false, UseLoopCountLimit=true)
  ]
  particleSpawnScript [
   &PartitionParticles()
   &InitializeParticle(
    Color=<rgba 0.3294 1 0.6039 1>
    Mesh Scale Mode=::NewEnumerator3
    Mesh Uniform Scale=0.99
    Sprite Size Mode=::NewEnumerator3
    Uniform Sprite Size=34.6994
    UsePositionOffset=false
    Write Material Random=true
   )
   &SpriteFacingAndAlignment(
    Sprite Facing=&Lerp_Vector(A=<0 1 0>, B=<0.2971 1 1>, Alpha=&MakeFloatFromInt(Int=$Particles.PartitionParticles.Partition))
    WriteAlignment=false
    WriteFacing=true
   ) {enabled=false}
  ]
  particleUpdateScript [
   &ParticleState(
    Kill Particles When Lifetime Has Elapsed=false
    Let Infinitely Lived Particles Die When Emitter Deactivates=true
    Loop Particles Lifetime=true
   )
   &SkeletalMeshLocation(
    MaskSpawnGroup=false
    Mesh Sampling Type=::NewEnumerator5
    Sampled Position Offset=<-0.1 -4.7 -0.04>
    Skeletal Mesh=$User.SkeletalMesh
    Socket Sampling Mode=::NewEnumerator5
   )
   /Niagara/Modules/Debug/SpriteBasedLine(
    Point A:NiagaraPosition=$Particles.SkeletalMeshLocation.SampledPosition
    Point B:NiagaraPosition=&Add_Vector(A=Linked() {linkedVariable {name=Particles.SkeletalMeshLocation.SampledPosition, type=NiagaraPosition}}, B=<0 0 0>)
   ) {enabled=false}
   &CameraOffset() {enabled=false}
   &ScaleSpriteSize(
    Scale Sprite Size Mode=::NewEnumerator0
    Uniform Scale Factor=&MakeCustomFloatFromBool(
     Bool=&SetBoolByIntComparison(A=$Particles.PartitionParticles.Partition, B=1, Comparison Type=::NewEnumerator0)
     True Float=0
     False Float=1
    )
   ) {enabled=false}
   &ScaleColor(
    Scale Alpha=&MakeFloatFromVector(
     Vector=&VectorFromFloat(Value=&Lerp_Float(
       A=0
       B=0.03
       Alpha=&ClampFloat(Float=&Add_Float(A=$System.Size scale, B=$System.Timeline.NormalizedTime))
      ))
     Channel=::NewEnumerator0
    )
    Scale RGB=&VectorFromFloat(Value=&MakeFloatFromVector(
      Vector=&VectorFromFloat(Value=&Lerp_Float(
        A=0
        B=1
        Alpha=&ClampFloat(Float=&Add_Float(A=$System.Size scale, B=$System.Timeline.NormalizedTime))
       ))
      Channel=::NewEnumerator0
     ))
    ScaleA=true
    ScaleRGB=true
   )
   /SetVariables(Particles.LightColor=&MultiplyLinearColor(
     Linear Color=&FNTimeOfDayColor(
      Night Color=<rgba 0.3294 1 0.6039 1>
      Morning Color=<rgba 0.3278 1 0.6038 1>
      Day Color=<rgba 0.3278 1 0.6038 1>
      Evening Color=<rgba 0.3278 1 0.6038 1>
     )
     Scale Factor=&MakeFloat4FromVectorAndFloat(
      "Vector (RGB)"=&Multiply_VectorByFloat(Vector=<1 1 1>, Float=51.8386)
      "Float (Alpha)"=&FN_FrontEndFloatModifier(Gameplay=&Lerp_Float(A=1, B=50, Alpha=$System.Timeline.NormalizedTime), Front End=1)
     )
    )) {enabled=false, isSetParameters=true}
   /SetVariables(Particles.LightPosition:NiagaraPosition=&Add_Vector(
     A=$Output.SkeletalMeshLocation.SampledPositionWithoutOffset
     B=&Lerp_Vector(A=<-25 -5 0>, B=<25 -5 0>, Alpha=&MakeFloatFromInt(Int=$Particles.PartitionParticles.Partition))
    )) {enabled=false, isSetParameters=true}
   &FN_Light_Attributes(Create Exponent Attribute=true, Create Radius Attribute=true, Exponent=3, Radius=32.4358) {enabled=false}
   &DynamicMaterialParameters(
    Index 0 Param 3=1
    Module.Param0WriteEnabled=true
    Module.Param1WriteEnabled=false
    Module.Param2WriteEnabled=true
    Module.Param3WriteEnabled=false
   )
  ]
  rendererClasses [&NiagaraMeshRendererProperties]
  renderers [{rendererClass=&NiagaraMeshRendererProperties, rendererIndex=0}]
 }
 SideWings001 {
  emitterSpawnScript [&FN_RibbonIDOffsetUtility()]
  emitterUpdateScript [
   &EmitterState()
   &FN_SpawnAnimTrail(Number of Anim Trails=4)
  ]
  particleSpawnScript [
   &FN_InitializeAnimTrailBoneIndecies(First Bone or Socket Index=1, Second Bone or Socket Index=2)
   &FN_InitializeAnimTrailBoneIndecies(AnimTrail Selection=1, First Bone or Socket Index=2, Second Bone or Socket Index=3) {module=FN_InitializeAnimTrailBoneIndecies001}
   &FN_InitializeAnimTrailBoneIndecies(AnimTrail Selection=2, First Bone or Socket Index=4, Second Bone or Socket Index=5) {module=FN_InitializeAnimTrailBoneIndecies003}
   &FN_InitializeAnimTrailBoneIndecies(AnimTrail Selection=3, First Bone or Socket Index=5, Second Bone or Socket Index=6) {module=FN_InitializeAnimTrailBoneIndecies002}
   &FN_InitializeAnimTrail(
    AnimTrailWidth=1.1
    Color=&FNTimeOfDayColor(Night Color=<rgba 1 1 1 0.15>, Morning Color=<rgba 1 1 1 0.28>, Day Color=<rgba 1 1 1 0.25>, Evening Color=<rgba 1 1 1 0.28>)
    Lifetime=0.06
   )
   &AddVelocity(Velocity=&MakeCustomVectorFromBool(Bool=$User.bIsFrontEnd, True Vector=<0 -2500 0>, False Vector=<0 2500 0>))
  ]
  particleUpdateScript [
   &SolveForcesAndVelocity(Clamp Velocity=false, Limit Acceleration=false)
   &FN_UpdateAnimTrail(SkeletalMesh=$User.SkeletalMesh)
   &InterpolateOverTime(
    Initial Vector Value:NiagaraPosition=$Particles.Position
    Target Vector Value:Vector3f=&AddVectorToPosition(Position=$Particles.Position, Vector=&TransformVector(Vector=<0 0 20>, Destination Space=::World))
    Variable Type=::NewEnumerator2
   ) {enabled=false}
   /SetVariables(Particles.Position=&AddVectorToPosition(
     Position=$Particles.Initial.FN_InitializeAnimTrail.Position Offset
     Vector=Linked() {linkedVariable {name=Particles.InterpolateOverTime.InterpolateOverTime_V3_.Moving Average, type=NiagaraPosition}}
    )) {enabled=false, isSetParameters=true}
   /SetVariables(Particles.RibbonU0Override=$Particles.NormalizedAge) {isSetParameters=true}
   &ScaleColor(
    Scale Alpha=&FN_FrontEndFloatModifier(Gameplay=&Lerp_Float(A=0, B=0.8, Alpha=$System.Timeline.NormalizedTime), Front End=1)
    ScaleA=true
    ScaleRGB=true
   )
  ]
  rendererClasses [&NiagaraRibbonRendererProperties, &NiagaraSpriteRendererProperties]
  renderers [
   {rendererClass=&NiagaraRibbonRendererProperties}
   {rendererClass=&NiagaraSpriteRendererProperties, rendererIndex=1}
  ]
 }
}
summary {
 systemName=NS_Glider_CirrusVine_Thrusters
 emitters [
  {emitterName=FakeLighting, rendererClasses [&NiagaraMeshRendererProperties]}
  {emitterName=AirThrusters001, rendererClasses [&NiagaraMeshRendererProperties]}
  {emitterName=AirThrusters, rendererClasses [&NiagaraMeshRendererProperties]}
  {emitterName=AirThrusters002, rendererClasses [&NiagaraMeshRendererProperties]}
  {emitterName=SideWings001, rendererClasses [&NiagaraRibbonRendererProperties, &NiagaraSpriteRendererProperties]}
  {emitterName=AirThrusters003, rendererClasses [&NiagaraSpriteRendererProperties]}
 ]
 userVariables [
  User.bIsFrontEnd {type=NiagaraBool}
  User.bIsFrontEndPreview {type=NiagaraBool}
  User.bIsFullyDeployed {type=NiagaraBool, defaultValue=true}
  User.ForwardDot {type=NiagaraFloat, defaultValue=1, description=""}
  User.RibbonIDOffset {type=NiagaraInt32}
  User.RightDot {type=NiagaraFloat, defaultValue=0, description=""}
  User.SkeletalMesh {
   type=NiagaraDataInterfaceSkeletalMesh
   defaultValue {
    struct=NiagaraExt_VariableValue_DataInterface
    value {dataInterface=&self:NiagaraDataInterfaceSkeletalMesh_1}
   }
  }
 ]
}
